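# Level Sequence for a CaptureTeam establishing shot: a roughly eight-second camera flythrough under a single camera cut.
# A spawned rail rig at a fixed transform animates its rail position from near the start to the end.
# The spawned Flycam cine camera has an attach track, presumably to that rail, and keyed rotation that swings the view during the move.
# Focus distance is keyed at fixed focal length, and a camera shake joins late in the shot.
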LevelSequence /PilgrimCore/CameraManager/Sequences/CaptureTeam/General/EstablishingShots/CT_EstablishingShot_3_CC
bindings [
 Flycam {
  id=#1
  possessed_class=CineCameraActor
  tracks [
   {
    type=MovieScene3DTransformTrack
    sections [{
      type=MovieScene3DTransformSection
      channels [
       Location.X {type=MovieSceneScriptingDoubleChannel, default=-27.9478}
       Location.Y {type=MovieSceneScriptingDoubleChannel, default=-50.1606}
       Location.Z {type=MovieSceneScriptingDoubleChannel, default=14.9447}
       Rotation.X {
        type=MovieSceneScriptingDoubleChannel
        default=360
        keys [[-35.615,0],[37.8975,-0.327],[162.2875,-5.5033],[232.5125,-25.9712],[332,-76.1742],[409.1275,-88.3692],[499.875,-95.7975]]
       }
       Rotation.Y {
        type=MovieSceneScriptingDoubleChannel
        default=32
        keys [[-35.615,8.6675],[37.8975,29.5261],[162.2875,49.24],[232.5125,45.5383],[332,-36.3637],[409.1275,-71.378],[499.875,-76.0346]]
       }
       Rotation.Z {
        type=MovieSceneScriptingDoubleChannel
        default=360
        keys [[-35.615,0],[37.8975,2.0547],[162.2875,-9.4828],[232.5125,-28.6906],[332,-57.0251],[409.1275,-77.6907],[499.875,-78.2249]]
       }
       Scale.X {type=MovieSceneScriptingDoubleChannel, default=1}
       Scale.Y {type=MovieSceneScriptingDoubleChannel, default=1}
       Scale.Z {type=MovieSceneScriptingDoubleChannel, default=1}
      ]
     }]
   }
   {type=MovieSceneSpawnTrack, sections [{type=MovieSceneSpawnSection, channels [None {type=MovieSceneScriptingBoolChannel, default=true}]}]}
   {type=MovieScene3DAttachTrack, sections [{type=MovieScene3DAttachSection, end_frame=503, start_frame=-37}]}
   {type=MovieSceneCameraShakeTrack, sections [{type=MovieSceneCameraShakeSection, end_frame=745, start_frame=387}]}
  ]
 }
 CameraComponent {
  id=#2
  parent_id=#1
  possessed_class=CineCameraComponent
  tracks [
   {
    type=MovieSceneFloatTrack
    display_name=Current Focal Length
    sections [{type=MovieSceneFloatSection, channels [None {type=MovieSceneScriptingFloatChannel, default=18}]}]
   }
   {
    type=MovieSceneFloatTrack
    display_name="Manual Focus Distance (Focus Settings)"
    sections [{
      type=MovieSceneFloatSection
      channels [None {type=MovieSceneScriptingFloatChannel, default=3642.5215, keys [[0,3642.5215],[231,385.9972],[424,722.7786]]}]
     }]
   }
   {
    type=MovieSceneFloatTrack
    display_name=Current Aperture
    sections [{type=MovieSceneFloatSection, channels [None {type=MovieSceneScriptingFloatChannel, default=5, const [5,1]}]}]
   }
  ]
 }
 CameraRig_Rail {
  id=#3
  possessed_class=CameraRig_Rail
  tracks [
   {
    type=MovieSceneFloatTrack
    display_name=Current Position on Rail
    sections [{
      type=MovieSceneFloatSection
      channels [None {type=MovieSceneScriptingFloatChannel, default=0.98, keys [[-377.9725,0.0664],[2.32,0.076],[427.225,0.98],[490,1]]}]
     }]
   }
   {type=MovieSceneSpawnTrack, sections [{type=MovieSceneSpawnSection, channels [None {type=MovieSceneScriptingBoolChannel, default=true}]}]}
   {
    type=MovieScene3DTransformTrack
    sections [{
      type=MovieScene3DTransformSection
      channels [
       Location.X {type=MovieSceneScriptingDoubleChannel, default=-9320}
       Location.Y {type=MovieSceneScriptingDoubleChannel, default=-536}
       Location.Z {type=MovieSceneScriptingDoubleChannel, default=2796}
       Rotation.X {type=MovieSceneScriptingDoubleChannel, default=180}
       Rotation.Y {type=MovieSceneScriptingDoubleChannel, default=-89.2473}
       Rotation.Z {type=MovieSceneScriptingDoubleChannel, default=180}
       Scale.X {type=MovieSceneScriptingDoubleChannel, default=1}
       Scale.Y {type=MovieSceneScriptingDoubleChannel, default=1}
       Scale.Z {type=MovieSceneScriptingDoubleChannel, default=1}
      ]
     }]
   }
  ]
 }
]
display_rate {denominator=1, numerator=60}
tick_resolution {denominator=1, numerator=24000}
tracks [{type=MovieSceneCameraCutTrack, sections [{type=MovieSceneCameraCutSection, end_frame=486, start_frame=0}]}]
