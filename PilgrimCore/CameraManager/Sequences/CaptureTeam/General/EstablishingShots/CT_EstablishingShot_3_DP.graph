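# Level sequence for CaptureTeam establishing shot 3 (DP), with one camera cut section and a cine camera on an attach track.
# The camera travels and rises while pitch and yaw shift, focal length lengthens slightly, and manual focus pulls from far to near, with camera shake layered on.
# Four female soldier mannequins play skeletal animations at fixed transforms; they and a slightly rotating cube have visibility set to false.
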LevelSequence /PilgrimCore/CameraManager/Sequences/CaptureTeam/General/EstablishingShots/CT_EstablishingShot_3_DP
bindings [
 Cine Camera Actor {
  id=#1
  possessed_class=CineCameraActor
  tracks [
   {type=MovieSceneSpawnTrack, sections [{type=MovieSceneSpawnSection, channels [None {type=MovieSceneScriptingBoolChannel, default=true}]}]}
   {
    type=MovieScene3DTransformTrack
    sections [{
      type=MovieScene3DTransformSection
      channels [
       Location.X {type=MovieSceneScriptingDoubleChannel, default=-1057.5359, keys [[0,-3534.916],[335,-5869.6665],[671,-6207.0649]]}
       Location.Y {type=MovieSceneScriptingDoubleChannel, default=-1812.9279, keys [[0,428.8243],[335,380.8561],[671,345.703]]}
       Location.Z {type=MovieSceneScriptingDoubleChannel, default=-270, keys [[0,-270],[335,-121.3126],[671,-109.803]]}
       Rotation.X {type=MovieSceneScriptingDoubleChannel, default=0, const [0,1]}
       Rotation.Y {type=MovieSceneScriptingDoubleChannel, default=14, keys [[0,39.8145],[208,8.5],[335,4],[640,14]]}
       Rotation.Z {type=MovieSceneScriptingDoubleChannel, default=165, keys [[0,180.3422],[262,181.6675],[671,165]]}
       Scale.X {type=MovieSceneScriptingDoubleChannel, default=1}
       Scale.Y {type=MovieSceneScriptingDoubleChannel, default=1}
       Scale.Z {type=MovieSceneScriptingDoubleChannel, default=1}
      ]
     }]
   }
   {type=MovieScene3DAttachTrack, sections [{type=MovieScene3DAttachSection, end_frame=806, start_frame=0}]}
  ]
 }
 Cube {
  id=#2
  possessed_class=FortStaticMeshActor
  tracks [
   {
    type=MovieScene3DTransformTrack
    sections [{
      type=MovieScene3DTransformSection
      channels [
       Location.X {type=MovieSceneScriptingDoubleChannel, default=-506.7651}
       Location.Y {type=MovieSceneScriptingDoubleChannel, default=-851.7299}
       Location.Z {type=MovieSceneScriptingDoubleChannel, default=809.9095}
       Rotation.X {type=MovieSceneScriptingDoubleChannel, default=0}
       Rotation.Y {type=MovieSceneScriptingDoubleChannel, default=0}
       Rotation.Z {type=MovieSceneScriptingDoubleChannel, default=0.05, keys [[264,0],[383,0.15],[661,0.05]]}
       Scale.X {type=MovieSceneScriptingDoubleChannel, default=1}
       Scale.Y {type=MovieSceneScriptingDoubleChannel, default=1}
       Scale.Z {type=MovieSceneScriptingDoubleChannel, default=1}
      ]
     }]
   }
   {
    type=MovieSceneVisibilityTrack
    sections [{type=MovieSceneVisibilitySection, channels [None {type=MovieSceneScriptingBoolChannel, default=false}]}]
   }
   {type=MovieSceneSpawnTrack, sections [{type=MovieSceneSpawnSection, channels [None {type=MovieSceneScriptingBoolChannel, default=true}]}]}
  ]
 }
 SoldierMannequin_Female {
  id=#3
  possessed_class=SoldierMannequin_Female_C
  tracks [
   {
    type=MovieScene3DTransformTrack
    sections [{
      type=MovieScene3DTransformSection
      channels [
       Location.X {type=MovieSceneScriptingDoubleChannel, default=-6893.6674, const [-6893.6675,1]}
       Location.Y {type=MovieSceneScriptingDoubleChannel, default=-470.6701, const [-470.6701,1]}
       Location.Z {type=MovieSceneScriptingDoubleChannel, default=626.9851, const [626.9851,1]}
       Rotation.X {type=MovieSceneScriptingDoubleChannel, default=0, const [0,1]}
       Rotation.Y {type=MovieSceneScriptingDoubleChannel, default=0, const [0,1]}
       Rotation.Z {type=MovieSceneScriptingDoubleChannel, default=-73.634, const [-73.6339,1]}
       Scale.X {type=MovieSceneScriptingDoubleChannel, default=1, const [1,1]}
       Scale.Y {type=MovieSceneScriptingDoubleChannel, default=1, const [1,1]}
       Scale.Z {type=MovieSceneScriptingDoubleChannel, default=1, const [1,1]}
      ]
     }]
   }
   {type=MovieSceneSkeletalAnimationTrack, sections [{type=MovieSceneSkeletalAnimationSection, end_frame=806, start_frame=0}]}
   {
    type=MovieSceneVisibilityTrack
    sections [{type=MovieSceneVisibilitySection, channels [None {type=MovieSceneScriptingBoolChannel, default=false}]}]
   }
  ]
 }
 CameraComponent {
  id=#4
  parent_id=#1
  possessed_class=CineCameraComponent
  tracks [
   {
    type=MovieSceneFloatTrack
    display_name=Current Focal Length
    sections [{type=MovieSceneFloatSection, channels [None {type=MovieSceneScriptingFloatChannel, default=20, keys [[0,20],[161,20],[362,25]]}]}]
   }
   {
    type=MovieSceneFloatTrack
    display_name="Manual Focus Distance (Focus Settings)"
    sections [{
      type=MovieSceneFloatSection
      channels [None {type=MovieSceneScriptingFloatChannel, default=200.2782, keys [[82,6825.3403],[208,1182.5627],[335,518.4426],[413,366.9508],[628,200.2782]]}]
     }]
   }
   {
    type=MovieSceneFloatTrack
    display_name=Current Aperture
    sections [{type=MovieSceneFloatSection, channels [None {type=MovieSceneScriptingFloatChannel, default=7, const [4,1]}]}]
   }
   {type=MovieSceneCameraShakeTrack, sections [{type=MovieSceneCameraShakeSection, end_frame=705, start_frame=-18}]}
  ]
 }
 "SoldierMannequin_Female (4)" {
  id=#5
  possessed_class=SoldierMannequin_Female_C
  tracks [
   {
    type=MovieScene3DTransformTrack
    sections [{
      type=MovieScene3DTransformSection
      channels [
       Location.X {type=MovieSceneScriptingDoubleChannel, default=-7427.9641, const [-7427.9639,1]}
       Location.Y {type=MovieSceneScriptingDoubleChannel, default=-609.7594, const [-609.7594,1]}
       Location.Z {type=MovieSceneScriptingDoubleChannel, default=778.4106, const [778.4106,1]}
       Rotation.X {type=MovieSceneScriptingDoubleChannel, default=0, const [0,1]}
       Rotation.Y {type=MovieSceneScriptingDoubleChannel, default=0, const [0,1]}
       Rotation.Z {type=MovieSceneScriptingDoubleChannel, default=-56.714, const [-56.714,1]}
       Scale.X {type=MovieSceneScriptingDoubleChannel, default=1, const [1,1]}
       Scale.Y {type=MovieSceneScriptingDoubleChannel, default=1, const [1,1]}
       Scale.Z {type=MovieSceneScriptingDoubleChannel, default=1, const [1,1]}
      ]
     }]
   }
   {type=MovieSceneSkeletalAnimationTrack, sections [{type=MovieSceneSkeletalAnimationSection, end_frame=783, start_frame=-128}]}
   {
    type=MovieSceneVisibilityTrack
    sections [{type=MovieSceneVisibilitySection, channels [None {type=MovieSceneScriptingBoolChannel, default=false}]}]
   }
  ]
 }
 "SoldierMannequin_Female (3)" {
  id=#6
  possessed_class=SoldierMannequin_Female_C
  tracks [
   {
    type=MovieScene3DTransformTrack
    sections [{
      type=MovieScene3DTransformSection
      channels [
       Location.X {type=MovieSceneScriptingDoubleChannel, default=-7202.4811, const [-7202.481,1]}
       Location.Y {type=MovieSceneScriptingDoubleChannel, default=-230.7822, const [-230.7822,1]}
       Location.Z {type=MovieSceneScriptingDoubleChannel, default=626.5436, const [626.5436,1]}
       Rotation.X {type=MovieSceneScriptingDoubleChannel, default=0, const [0,1]}
       Rotation.Y {type=MovieSceneScriptingDoubleChannel, default=0, const [0,1]}
       Rotation.Z {type=MovieSceneScriptingDoubleChannel, default=-68.9807, const [-68.9807,1]}
       Scale.X {type=MovieSceneScriptingDoubleChannel, default=1, const [1,1]}
       Scale.Y {type=MovieSceneScriptingDoubleChannel, default=1, const [1,1]}
       Scale.Z {type=MovieSceneScriptingDoubleChannel, default=1, const [1,1]}
      ]
     }]
   }
   {type=MovieSceneSkeletalAnimationTrack, sections [{type=MovieSceneSkeletalAnimationSection, end_frame=768, start_frame=0}]}
   {
    type=MovieSceneVisibilityTrack
    sections [{type=MovieSceneVisibilitySection, channels [None {type=MovieSceneScriptingBoolChannel, default=false}]}]
   }
  ]
 }
 "SoldierMannequin_Female (2)" {
  id=#7
  possessed_class=SoldierMannequin_Female_C
  tracks [
   {
    type=MovieScene3DTransformTrack
    sections [{
      type=MovieScene3DTransformSection
      channels [
       Location.X {type=MovieSceneScriptingDoubleChannel, default=-7099.0414, const [-7099.0415,1]}
       Location.Y {type=MovieSceneScriptingDoubleChannel, default=-779.9541, const [-779.9542,1]}
       Location.Z {type=MovieSceneScriptingDoubleChannel, default=677.6108, const [677.6108,1]}
       Rotation.X {type=MovieSceneScriptingDoubleChannel, default=0, const [0,1]}
       Rotation.Y {type=MovieSceneScriptingDoubleChannel, default=0, const [0,1]}
       Rotation.Z {type=MovieSceneScriptingDoubleChannel, default=-62.0882, const [-62.0882,1]}
       Scale.X {type=MovieSceneScriptingDoubleChannel, default=1, const [1,1]}
       Scale.Y {type=MovieSceneScriptingDoubleChannel, default=1, const [1,1]}
       Scale.Z {type=MovieSceneScriptingDoubleChannel, default=1, const [1,1]}
      ]
     }]
   }
   {type=MovieSceneSkeletalAnimationTrack, sections [{type=MovieSceneSkeletalAnimationSection, end_frame=768, start_frame=0}]}
   {
    type=MovieSceneVisibilityTrack
    sections [{type=MovieSceneVisibilitySection, channels [None {type=MovieSceneScriptingBoolChannel, default=false}]}]
   }
  ]
 }
]
display_rate {denominator=1, numerator=60}
tick_resolution {denominator=1, numerator=24000}
tracks [{type=MovieSceneCameraCutTrack, sections [{type=MovieSceneCameraCutSection, end_frame=752, start_frame=0}]}]
