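# Level sequence for a CaptureTeam establishing shot: a five-second cine camera move under a single camera cut.
# The camera follows a keyed arc that dips then climbs, panning steadily while rolling from one side to the other at a downward pitch.
# Transform keys extend slightly past both ends of the cut, so the camera is already moving when the shot starts and ends.
# Focal length, manual focus distance and aperture stay constant.
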LevelSequence /PilgrimCore/CameraManager/Sequences/CaptureTeam/General/EstablishingShots/CT_EstablishingShot_4_CC
bindings [
 CameraComponent {
  id=#1
  parent_id=#2
  possessed_class=CineCameraComponent
  tracks [
   {
    type=MovieSceneFloatTrack
    display_name=Current Focal Length
    sections [{type=MovieSceneFloatSection, channels [None {type=MovieSceneScriptingFloatChannel, default=35}]}]
   }
   {
    type=MovieSceneFloatTrack
    display_name="Manual Focus Distance (Focus Settings)"
    sections [{type=MovieSceneFloatSection, channels [None {type=MovieSceneScriptingFloatChannel, default=3000}]}]
   }
   {
    type=MovieSceneFloatTrack
    display_name=Current Aperture
    sections [{type=MovieSceneFloatSection, channels [None {type=MovieSceneScriptingFloatChannel, default=0.1}]}]
   }
  ]
 }
 Cine Camera Actor {
  id=#2
  possessed_class=CineCameraActor
  tracks [
   {
    type=MovieScene3DTransformTrack
    sections [{
      type=MovieScene3DTransformSection
      channels [
       Location.X {
        type=MovieSceneScriptingDoubleChannel
        default=-5676.9975
        keys [[-9.54,-5603.5811],[105,-4641.6797],[188.2075,-4236.4971],[309.63,-4188.5117]]
       }
       Location.Y {type=MovieSceneScriptingDoubleChannel, default=-666.0143, keys [[-9.54,1439.0977],[105,475.6084],[188.2075,-643.3018],[309.63,-1626.1871]]}
       Location.Z {type=MovieSceneScriptingDoubleChannel, default=767.2663, keys [[-9.54,1596.4071],[105,1449.1467],[188.2075,1594.1331],[309.63,1817.1783]]}
       Rotation.X {type=MovieSceneScriptingDoubleChannel, default=0, keys [[-9.54,-19.3896],[105,7.9866],[188.2075,19.117],[309.63,23.0723]]}
       Rotation.Y {type=MovieSceneScriptingDoubleChannel, default=-4.1835, thinned_from=4, keys [[-9.54,-17.489],[105,-15.9479],[309.63,-19.0548]]}
       Rotation.Z {type=MovieSceneScriptingDoubleChannel, default=158.2756, keys [[-9.54,217.2458],[105,201.0813],[188.2075,177.5401],[309.63,158.2756]]}
       Scale.X {type=MovieSceneScriptingDoubleChannel, default=1, const [1,4]}
       Scale.Y {type=MovieSceneScriptingDoubleChannel, default=1, const [1,4]}
       Scale.Z {type=MovieSceneScriptingDoubleChannel, default=1, const [1,4]}
      ]
     }]
   }
   {type=MovieSceneSpawnTrack, sections [{type=MovieSceneSpawnSection, channels [None {type=MovieSceneScriptingBoolChannel, default=true}]}]}
  ]
 }
]
display_rate {denominator=1, numerator=60}
tick_resolution {denominator=1, numerator=24000}
tracks [{type=MovieSceneCameraCutTrack, sections [{type=MovieSceneCameraCutSection, end_frame=300, start_frame=0}]}]
