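# Level Sequence for a capture-team establishing shot (variant 5), played through a single spawned Cine Camera Actor.
# A transform track glides the camera forward along negative X with a slight rise, while a strong roll eases to near level.
# The camera component holds a wide lens, manual focus distance and narrow aperture constant.
# One camera cut section spans the shot and runs on after the last transform key.
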LevelSequence /PilgrimCore/CameraManager/Sequences/CaptureTeam/General/EstablishingShots/CT_EstablishingShot_5_CC
bindings [
 CameraComponent {
  id=#1
  parent_id=#2
  possessed_class=CineCameraComponent
  tracks [
   {
    type=MovieSceneFloatTrack
    display_name=Current Focal Length
    sections [{type=MovieSceneFloatSection, channels [None {type=MovieSceneScriptingFloatChannel, default=25}]}]
   }
   {
    type=MovieSceneFloatTrack
    display_name="Manual Focus Distance (Focus Settings)"
    sections [{type=MovieSceneFloatSection, channels [None {type=MovieSceneScriptingFloatChannel, default=4641.4092}]}]
   }
   {
    type=MovieSceneFloatTrack
    display_name=Current Aperture
    sections [{type=MovieSceneFloatSection, channels [None {type=MovieSceneScriptingFloatChannel, default=16}]}]
   }
  ]
 }
 Cine Camera Actor {
  id=#2
  possessed_class=CineCameraActor
  tracks [
   {
    type=MovieScene3DTransformTrack
    sections [{
      type=MovieScene3DTransformSection
      channels [
       Location.X {type=MovieSceneScriptingDoubleChannel, default=7811.984, keys [[-22.7625,-952.3864],[339.385,-2587.4465]]}
       Location.Y {type=MovieSceneScriptingDoubleChannel, default=-2852.5804, keys [[-22.7625,-1165.6534],[339.385,-1188.9282]]}
       Location.Z {type=MovieSceneScriptingDoubleChannel, default=2196.0551, keys [[-22.7625,882.502],[339.385,934.7319]]}
       Rotation.X {type=MovieSceneScriptingDoubleChannel, default=0, keys [[-22.7625,-15.0119],[338.385,1.831],[339.385,0.96]]}
       Rotation.Y {type=MovieSceneScriptingDoubleChannel, default=0, keys [[-22.7625,-0.3257],[338.385,-1.8674],[339.385,-4.4731]]}
       Rotation.Z {type=MovieSceneScriptingDoubleChannel, default=180, keys [[-22.7625,175.2307],[338.385,173.2964],[339.385,174.6604]]}
       Scale.X {type=MovieSceneScriptingDoubleChannel, default=1, const [1,2]}
       Scale.Y {type=MovieSceneScriptingDoubleChannel, default=1, const [1,2]}
       Scale.Z {type=MovieSceneScriptingDoubleChannel, default=1, const [1,2]}
      ]
     }]
   }
   {type=MovieSceneSpawnTrack, sections [{type=MovieSceneSpawnSection, channels [None {type=MovieSceneScriptingBoolChannel, default=true}]}]}
  ]
 }
]
display_rate {denominator=1, numerator=60}
tick_resolution {denominator=1, numerator=24000}
tracks [{type=MovieSceneCameraCutTrack, sections [{type=MovieSceneCameraCutSection, end_frame=600, start_frame=0}]}]
