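# Level sequence for a capture-team establishing shot (number 7), a ten-second cut at 60 fps.
# A spawned Cine Camera Actor makes a single descending move with changing rotation, keyed from just before the cut starts to a little past its midpoint.
# The camera component holds a wide focal length and fixed aperture while manual focus distance pulls nearer across the whole shot.
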LevelSequence /PilgrimCore/CameraManager/Sequences/CaptureTeam/General/EstablishingShots/CT_EstablishingShot_7_CC
bindings [
 CameraComponent {
  id=#1
  parent_id=#2
  possessed_class=CineCameraComponent
  tracks [
   {
    type=MovieSceneFloatTrack
    display_name=Current Focal Length
    sections [{type=MovieSceneFloatSection, channels [None {type=MovieSceneScriptingFloatChannel, default=22, const [22,1]}]}]
   }
   {
    type=MovieSceneFloatTrack
    display_name="Manual Focus Distance (Focus Settings)"
    sections [{type=MovieSceneFloatSection, channels [None {type=MovieSceneScriptingFloatChannel, default=4211.0283, keys [[0,4211.0283],[599,2855.9192]]}]}]
   }
   {
    type=MovieSceneFloatTrack
    display_name=Current Aperture
    sections [{type=MovieSceneFloatSection, channels [None {type=MovieSceneScriptingFloatChannel, default=8}]}]
   }
  ]
 }
 Cine Camera Actor {
  id=#2
  possessed_class=CineCameraActor
  tracks [
   {
    type=MovieScene3DTransformTrack
    sections [{
      type=MovieScene3DTransformSection
      channels [
       Location.X {type=MovieSceneScriptingDoubleChannel, default=7811.984, keys [[-19.025,-6311.6519],[348.35,-8647.832]]}
       Location.Y {type=MovieSceneScriptingDoubleChannel, default=-2852.5804, keys [[-19.025,3647.2361],[348.35,1181.2252]]}
       Location.Z {type=MovieSceneScriptingDoubleChannel, default=2196.0551, keys [[-19.025,1558.3246],[348.35,872.7642]]}
       Rotation.X {type=MovieSceneScriptingDoubleChannel, default=0, keys [[-19.025,1.4567],[348.35,-1.1019]]}
       Rotation.Y {type=MovieSceneScriptingDoubleChannel, default=0, keys [[-19.025,-6.8996],[348.35,2.5239]]}
       Rotation.Z {type=MovieSceneScriptingDoubleChannel, default=180, keys [[-19.025,263.0869],[348.35,311.7735]]}
       Scale.X {type=MovieSceneScriptingDoubleChannel, default=1, const [1,2]}
       Scale.Y {type=MovieSceneScriptingDoubleChannel, default=1, const [1,2]}
       Scale.Z {type=MovieSceneScriptingDoubleChannel, default=1, const [1,2]}
      ]
     }]
   }
   {type=MovieSceneSpawnTrack, sections [{type=MovieSceneSpawnSection, channels [None {type=MovieSceneScriptingBoolChannel, default=true}]}]}
  ]
 }
]
display_rate {denominator=1, numerator=60}
tick_resolution {denominator=1, numerator=24000}
tracks [{type=MovieSceneCameraCutTrack, sections [{type=MovieSceneCameraCutSection, end_frame=600, start_frame=0}]}]
